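# Translucent lit glass material for the Secretooze vehicle wheel cosmetic, faking animated bubbly liquid tinted by CustomColor.
# Dark, base and light shades derived from that color tint waves built from counter-panning liquid textures, cloud distortion and reflection-vector caustics.
# Two bump-offset, panning bubble layers fake depth, and a Fresnel rim perturbed by liquid and cloud normals adds glow.
# Material attributes carry color, emissive and parameterized opacity through shared vehicle damage and de-rez functions.
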Material /VehicleCosmetics/Wheels/Secretooze/Materials/M_SecretOoze_Glass
graph {
 properties {
  bAutomaticallySetUsageInEditor=false
  bUseMaterialAttributes=true
  bUsedWithMeshDeformer=true
  bUsedWithSkeletalMesh=true
  blendMode=BLEND_Translucent
  usedShadingModels=Default Lit
 }
 nodes [
  n1 = NamedRerouteUsage()
  n2 = ScalarParameter() {defaultValue=0.25, parameterName=Opacity}
  n3 = NamedRerouteUsage()
  MaterialFunctionCall(In=MaterialFunctionCall(Material=SetMaterialAttributes!(MaterialAttributes, Emissive Color=@n1, Opacity=@n2, Base Color=@n3) {attributeSetTypes [#1, #2, #3]}) {materialFunction=&MF_BaseVehicle_Damage}.Result) {materialFunction=&MF_VehicleDeRez_Transition}
  n4 = TextureCoordinate() {uTiling=0.25, vTiling=0.15}
  n5 = ComponentMask(None=TextureSample(UVs=Panner(Coordinate=@n4) {speedY=-0.01}) {texture=/VehicleCosmetics/Wheels/Secretooze/Textures/T_GenericCloud_N}.RGB) {g=true, r=true}
  NamedRerouteDeclaration(None=Multiply(A=@n5, B=Constant() {r=0.3})) {name=WaterPanner, nodeColor=<rgba 0 1 0.6953 1>}
  n6 = VectorParameter() {parameterName=CustomColor, defaultValue=<rgba 0.226 0.972 0.15 0>}
  n7 = SquareRoot(None=Multiply(A=@n6.RGB, B=@n6.RGB))
  n8 = Subtract(A=ConstantBiasScale(None=@n7) {bias=-0.2, scale=8}, B=ConstantBiasScale(None=@n7) {bias=-1})
  n9 = ConstantBiasScale(None=@n8) {bias=-0.05, scale=1.2}
  n10 = Clamp(None=Multiply(A=Clamp(None=@n9), B=Constant() {r=0.4}))
  NamedRerouteDeclaration(None=@n10) {name=BaseColor, nodeColor=<rgba 0 1 0.1328 1>}
  n11 = Constant() {r=0.75}
  n12 = Clamp(None=ConstantBiasScale(None=Multiply(A=@n6.RGB, B=@n11)) {bias=-0.25, scale=0.4})
  NamedRerouteDeclaration(None=@n12) {name=DarkColor, nodeColor=<rgba 1 0 0.1875 1>}
  n13 = Clamp(None=@n8)
  NamedRerouteDeclaration(None=@n13) {name=LightColor, nodeColor=<rgba 0 0.5078 1 1>}
  n14 = ComponentMask(None=ReflectionVectorWS()) {g=true, r=true}
  n15 = Constant() {r=0.5}
  n16 = Multiply(A=TextureSample(UVs=Multiply(A=@n14, B=@n15)) {texture=/VehicleCosmetics/Wheels/Secretooze/Textures/T_Caustics_D}.RGB, B=Constant() {r=1})
  n17 = Multiply(A=Constant() {r=6}, B=Time())
  n18 = Constant() {r=1}
  n19 = TextureCoordinate() {uTiling=2, vTiling=4}
  n20 = Add(A=BumpOffset(Coordinate=Multiply(A=@n18, B=@n19), Height=Constant() {r=-5}), B=Multiply(A=@n5, B=Constant() {r=0.5}))
  n21 = LinearInterpolate(
   A=TextureSample(UVs=Panner(Coordinate=@n20, Time=@n17) {speedX=-0.01, speedY=0.005}) {texture=/VehicleCosmetics/Wheels/Secretooze/Textures/T_Liquid_Tile_N}.RGB
   B=TextureSample(UVs=Panner(Coordinate=@n20, Time=@n17) {speedX=0.01, speedY=-0.005}) {texture=&T_Liquid_Tile_N}.RGB
   Alpha=Constant() {r=0.8}
  )
  n22 = Desaturation(None=@n21)
  n23 = Clamp(None=Multiply(A=ConstantBiasScale(None=@n22) {bias=-0.3, scale=1}, B=@n22))
  n24 = Add(A=@n22, B=Add(A=Multiply(A=@n16, B=@n23), B=@n10))
  n25 = Multiply(A=@n13, B=Constant() {r=2})
  NamedRerouteDeclaration(None=Multiply(A=Clamp(None=@n24), B=LinearInterpolate(A=@n12, B=@n25, Alpha=@n23))) {name=Waves, nodeColor=<rgba 0.8934 1 0.49 1>}
  n26 = TextureCoordinate() {uTiling=3, vTiling=3}
  n27 = Constant() {r=-0.5}
  n28 = Multiply(A=Constant() {r=0.35}, B=Time())
  n29 = TextureSample(UVs=Panner(Coordinate=BumpOffset(Coordinate=@n26, Height=@n27), Time=@n28) {speedX=-0.01, speedY=-0.125}) {texture=/VehicleCosmetics/Wheels/Octo/Textures/Bubbles}
  n30 = Multiply(A=@n29.RGB, B=Multiply(A=@n29.A, B=Constant() {r=3}))
  n31 = BumpOffset(Coordinate=TextureCoordinate() {uTiling=3, vTiling=3}, Height=Constant() {r=-1})
  n32 = NamedRerouteUsage()
  n33 = TextureSample(UVs=Panner(Coordinate=Add(A=@n31, B=@n32), Time=@n28) {speedX=0.015, speedY=-0.1}) {texture=&Bubbles}
  n34 = Multiply(A=@n33.RGB, B=Multiply(A=Desaturation(None=@n33.RGB), B=Constant() {r=3}))
  n35 = Multiply(A=NamedRerouteUsage(), B=Multiply(A=LinearInterpolate(A=@n30, B=@n34), B=Constant() {r=3}))
  NamedRerouteDeclaration(None=Add(A=NamedRerouteUsage(), B=@n35)) {name=WavesBubbles, nodeColor=<rgba 0.5196 1 0 1>}
  n36 = Add(
   A=Multiply(A=@n21, B=Constant() {r=0.15})
   B=TextureSample(UVs=TextureCoordinate()) {texture=&T_GenericCloud_N}.RGB
  )
  n37 = Clamp(None=MaterialFunctionCall!(Normal=Clamp(None=@n36), Exponent) {materialFunction=/VehicleCosmetics/Wheels/Cola_Holographic/Materials/MF_Fresnel_Classic}.Result)
  n38 = Constant() {r=1.5}
  n39 = Multiply(
   A=VectorParameter() {parameterName=CustomColor, defaultValue=<rgba 0.226 0.972 0.15 0>}.RGB
   B=Multiply(A=Power(Base=@n37, Exp=@n38), B=Constant() {r=0.4})
  )
  NamedRerouteDeclaration(None=Add(A=@n39, B=NamedRerouteUsage())) {name=Emissive, nodeColor=<rgba 1 0 0.4911 1>}
  NamedRerouteDeclaration(None=@n35) {name=Bubbles, nodeColor=<rgba 0.2942 0.8118 1 1>}
  NamedRerouteDeclaration(None=Add(A=@n39, B=NamedRerouteUsage())) {name=BC, nodeColor=<rgba 0.5 0 1 1>}
 ]
}
