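# Skinned wheel material for the SecretOoze vehicle cosmetic, built on the SecretOoze simple wheel parent function.
# Time-driven sine and cosine waves, masked by vertex green, add a mostly-Z World Position Offset wobble.
# Vertex red blends parent normals detailed with brushed-metal and matte noise textures.
# These overrides feed vehicle damage, then de-rez transition functions; an extra generic wheel parent call is unconnected.
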Material /VehicleCosmetics/Wheels/Secretooze/Materials/M_Wheel_SecretOoze
graph {
 properties {
  bAutomaticallySetUsageInEditor=false
  bUseMaterialAttributes=true
  bUsedWithMeshDeformer=true
  bUsedWithSkeletalMesh=true
  pixelDepthOffsetMode=PDOM_Legacy
  previewMesh=/VehicleCosmetics/Wheels/Secretooze/Meshes/SK_Wheel_SecretOoze
  usedShadingModels=Default Lit
 }
 nodes [
  n1 = MaterialFunctionCall() {materialFunction=&/MF_WheelParent_Simple_SecretOoze}
  n2 = NamedRerouteUsage()
  n3 = NamedRerouteUsage()
  MaterialFunctionCall(In=MaterialFunctionCall(Material=SetMaterialAttributes!(MaterialAttributes=@n1.MaterialAttributes, World Position Offset=@n2, Normal=@n3) {attributeSetTypes [#1, #2]}) {materialFunction=&MF_BaseVehicle_Damage}.Result) {materialFunction=&MF_VehicleDeRez_Transition}
  MaterialFunctionCall() {materialFunction=&MF_WheelParent_Simple}
  n4 = Time()
  n5 = AppendVector(A=AppendVector(A=Sine(None=@n4) {period=9}, B=Cosine(None=@n4) {period=9}), B=Sine(None=@n4) {period=13})
  n6 = Constant3Vector() {constant=<rgba 0.15 0.2 5 0>}
  NamedRerouteDeclaration(None=Multiply(A=VertexColor().G, B=Multiply(A=@n5, B=@n6))) {name=WPO, nodeColor=<rgba 1 0 0.2344 1>}
  NamedRerouteDeclaration(None=@n1.Normal_IN) {name=Normal_IN, nodeColor=<rgba 0.4197 0.2836 1 1>}
  n7 = NamedRerouteUsage()
  n8 = TextureCoordinate() {vTiling=10}
  n9 = ScalarParameter() {defaultValue=1, group=NormalDetails, parameterName=BrushedMetalTile}
  n10 = Multiply(
   A=TextureSample(UVs=Multiply(A=@n8, B=@n9)) {samplerType=SAMPLERTYPE_Normal, texture=/VehicleCosmetics/Textures/UniversalTextures/T_BrushedMetal_N}.RGB
   B=Constant3Vector() {constant=<rgba 0.1 0.1 0 0>}
  )
  n11 = Multiply(A=NamedRerouteUsage(), B=Constant3Vector() {constant=<rgba 0.625 0.625 1 0>})
  n12 = TextureCoordinate()
  n13 = ScalarParameter() {defaultValue=1, group=NormalDetails, parameterName=NoiseTile}
  n14 = Multiply(
   A=TextureSample(UVs=Multiply(A=@n12, B=@n13)) {samplerType=SAMPLERTYPE_Normal, texture=/VehicleCosmetics/SharedTextures/T_Matte_N}.RGB
   B=Constant3Vector() {constant=<rgba 0.375 0.375 0 0>}
  )
  NamedRerouteDeclaration(None=LinearInterpolate(A=Add(A=@n7, B=@n10), B=Add(A=@n11, B=@n14), Alpha=VertexColor().R)) {name=Normal, nodeColor=<rgba 0.3509 0 1 1>}
  NamedRerouteDeclaration(None=@n1.Normal_Tire) {name=Normal_Tire, nodeColor=<rgba 0.208 0.4882 1 1>}
 ]
}
